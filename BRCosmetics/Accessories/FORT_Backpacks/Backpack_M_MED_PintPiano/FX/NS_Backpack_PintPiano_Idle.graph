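# Idle Niagara effect for the PintPiano backpack cosmetic: a looping, flickering glow that responds to reactivity.
# One emitter, NE_MuzzleFlash, bursts two spawn groups at different offsets and sizes, looping particle lifetime so they persist.
# System.ReactivityIntensity blends dim warm tones toward brighter blue, raising size, flare length and flicker speed.
# Curve-driven alpha and time-of-day scaling shape flare, dot and fake-light attributes alongside six sprite renderers.
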NiagaraSystem /BRCosmetics/Accessories/FORT_Backpacks/Backpack_M_MED_PintPiano/FX/NS_Backpack_PintPiano_Idle
dependencies {
 usedDataInterfaces [&NiagaraDataInterfaceCurve]
 usedDynamicInputs [
  &MakeCustomFloatFromBool
  &Divide_Float
  &Subtract_Float
  &FN_FrontEndFloatModifier
 ]
 usedModules [
  &SystemState
  &self:SystemSpawnScript.SystemScriptSource.SystemScriptGraph.NiagaraNodeAssignment_3.SetVariables_#1
 ]
}
emitters {NE_MuzzleFlash {
  emitterUpdateScript [
   &EmitterState(
    Inactive Response=::NewEnumerator1
    Life Cycle Mode=::NewEnumerator1
    Loop Behavior=::NewEnumerator1
    Loop Duration Mode=::NewEnumerator1
    UseLoopDelay=false
   )
   &SpawnBurst_Instantaneous(Use Spawn Probability=false, UseLoopCountLimit=false)
   &SpawnBurst_Instantaneous(Spawn Group=1, Use Spawn Probability=false, UseLoopCountLimit=false) {module=SpawnBurst_Instantaneous001}
  ]
  particleSpawnScript [
   &InitializeParticle(
    Color Mode=::NewEnumerator0
    Lifetime=0.3
    Position Offset=&MergeHelper_VectorSwitch(0=<0 -7.5 0.3>, 1=<0 -11 0.3>, 2=?, 3=?, Index=&SpawnGroup(), Num=::NewEnumerator0)
    Sprite Size Mode=::NewEnumerator3
    Uniform Sprite Size=&MergeHelper_FloatSwitch(0=10, 1=27, 2=?, 3=?, Index=&SpawnGroup(), Num=::NewEnumerator0)
    UsePositionOffset=true
    Write Material Random=true
   )
   &InitialMeshOrientation(Orientation Axis=<0 0 1>, Orientation Vector=<0 0 1>, Rotation=<0 0.25 0>, Use Orientation Vector=true, Use Rotation Vector=true)
   &CameraOffset()
   /SetVariables(Particles.VisibilityTag=&SpawnGroup()) {isSetParameters=true}
  ]
  particleUpdateScript [
   &ParticleState(Kill Particles When Lifetime Has Elapsed=false, Loop Particles Lifetime=true)
   &ScaleSpriteSize(
    Scale Sprite Size Mode=::NewEnumerator0
    Uniform Scale Factor=&Lerp_Float(A=0.5, B=1.7, Alpha=$System.ReactivityIntensity)
   )
   &Color(Color=&MultiplyLinearColor(
     Linear Color=&LerpLinearColors(StartColor=<rgba 0.5 0.3542 0.15 1>, EndColor=<rgba 0.14 0.2333 0.7 1>, LerpFactor=$System.ReactivityIntensity)
     Scale Factor=&MakeVector4FromVectorAndFloat(
      "Vector (RGB)"=&VectorFromFloat(Value=&FNTimeOfDayScalar(Night Scale=0.8, Morning Scale=0.6, Evening Scale=0.6))
      "Float (Alpha)"=&FloatFromCurve(
       FloatCurve=DataInterface() {propertyValues {Curve {
          defaultValue=0
          keys [
           {interpMode=RCIM_Cubic, tangentMode=RCTM_User, time=0, value=0.4996}
           {arriveTangent=0.1191, interpMode=RCIM_Cubic, leaveTangent=0.1191, time=0.2, value=0.8952}
           {arriveTangent=0.3494, interpMode=RCIM_Cubic, leaveTangent=0.3494, time=0.4, value=0.5472}
           {interpMode=RCIM_Cubic, tangentMode=RCTM_User, time=0.5, value=1}
           {arriveTangent=0.0382, interpMode=RCIM_Cubic, leaveTangent=0.0382, time=0.6, value=0.6168}
           {arriveTangent=-0.2931, interpMode=RCIM_Cubic, leaveTangent=-0.2931, time=0.8, value=1.0115}
           {tangentMode=RCTM_User, time=1, value=0.4996}
          ]
         }}}
       CurveIndex=&Multiply_Float(A=$Particles.NormalizedAge, B=&Lerp_Float(A=0.15, B=0.5, Alpha=$System.ReactivityIntensity))
      )
     )
    ))
   /SetVariables(Particles.LongFlareSize:Vector2f=&Multiply_Vector2D(A=$Particles.SpriteSize, B=&Lerp_Vector2(A=<0.2 0.05>, B=<1 0.07>, Alpha=$System.ReactivityIntensity))) {isSetParameters=true}
   &FN_Sprite_Attributes(
    Color=&MultiplyLinearColor(
     Linear Color=$Particles.Color
     Scale Factor=&MakeVector4_FromLinearColor(LinearColor=<rgba 1.5 1.5 1.5 0.75>)
    )
    Create Color Attribute=true
    Create SpriteSize Attribute=true
    SpriteSize=&Multiply_Vector2D_ByFloat(Float=2, Vector2D=$Particles.SpriteSize)
    Write to Module Namespace=true
   )
   &DynamicMaterialParameters(
    Index 3 Param 1=1
    Index 3 Param 4=&Multiply_Float(A=$Engine.System.Age, B=-2)
    Module.Index3Float0Write=true
    Module.Index3Float1Write=true
    Module.Index3Float2Write=true
    Module.Index3Float3Write=true
    Write Parameter Index 0=::NewEnumerator0
    Write Parameter Index 3=::NewEnumerator1
   )
   /SetVariables(
    Particles.DotColor:LinearColor=&MultiplyLinearColor(Linear Color=$Particles.Color, Scale Factor=<40 40 40 10>)
    Particles.DotSize:Vector2f=&Multiply_Vector2D_ByFloat(Float=&Lerp_Float(A=0.5, B=0.8, Alpha=$System.ReactivityIntensity), Vector2D=<1.5 1.5>)
   ) {isSetParameters=true}
   /SetVariables(Particles.FakeLightColor:LinearColor=&MultiplyLinearColor(
     Linear Color=&LerpLinearColors(StartColor=<rgba 0.05 0.0358 0.0075 1>, EndColor=<rgba 0.4 0.6667 2 1>, LerpFactor=$System.ReactivityIntensity)
     Scale Factor=&MakeVector4FromVectorAndFloat(
      "Vector (RGB)"=&VectorFromFloat(Value=&FNTimeOfDayScalar(Night Scale=0.6, Morning Scale=0.8, Evening Scale=0.8))
      "Float (Alpha)"=&FloatFromCurve(
       FloatCurve=DataInterface() {propertyValues {Curve {
          defaultValue=0
          keys [
           {interpMode=RCIM_Cubic, tangentMode=RCTM_User, time=0, value=0.604}
           {arriveTangent=0.1191, interpMode=RCIM_Cubic, leaveTangent=0.1191, time=0.2, value=0.8952}
           {arriveTangent=0.3494, interpMode=RCIM_Cubic, leaveTangent=0.3494, time=0.4, value=0.6516}
           {interpMode=RCIM_Cubic, tangentMode=RCTM_User, time=0.5, value=1}
           {arriveTangent=0.0382, interpMode=RCIM_Cubic, leaveTangent=0.0382, time=0.6, value=0.8385}
           {arriveTangent=-0.5865, interpMode=RCIM_Cubic, leaveTangent=-0.5865, time=0.8, value=1.0115}
           {tangentMode=RCTM_User, time=1, value=0.604}
          ]
         }}}
       CurveIndex=&Multiply_Float(A=$Particles.NormalizedAge, B=&Lerp_Float(A=0.3, B=2, Alpha=$System.ReactivityIntensity))
      )
     )
    )) {isSetParameters=true}
  ]
  rendererClasses [&NiagaraSpriteRendererProperties]
  renderers [
   {rendererClass=&NiagaraSpriteRendererProperties}
   {rendererClass=&NiagaraSpriteRendererProperties, rendererIndex=1}
   {rendererClass=&NiagaraSpriteRendererProperties, rendererIndex=2}
   {rendererClass=&NiagaraSpriteRendererProperties, rendererIndex=3}
   {rendererClass=&NiagaraSpriteRendererProperties, rendererIndex=4}
   {rendererClass=&NiagaraSpriteRendererProperties, rendererIndex=5}
  ]
 }}
summary {
 systemName=NS_Backpack_PintPiano_Idle
 emitters [{emitterName=NE_MuzzleFlash, rendererClasses [&NiagaraSpriteRendererProperties]}]
 userVariables [
  User.bIsFrontEnd {type=NiagaraBool}
  User.bIsFrontEndPreview {type=NiagaraBool}
  User.DebugReactivity {type=NiagaraBool}
  User.ForwardDot {type=NiagaraFloat, defaultValue=1, description=""}
  User.ReactivityTime {type=NiagaraFloat, defaultValue=0, description=""}
 ]
}
