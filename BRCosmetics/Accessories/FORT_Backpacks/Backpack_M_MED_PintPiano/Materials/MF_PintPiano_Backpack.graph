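# Material function for the PintPiano backpack cosmetic that sets a flickering, reactive emissive glow on incoming material attributes.
# A sphere mask on pre-skinned local position, interpolated from the vertex shader, shapes a colored glow flickered by two sines at different rates.
# A TimeStart-driven envelope blends toward a reactivity glow color, with a static switch for a debug loop.
# Inverse Fresnel and masks, including one for the glass panel, limit the result.
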MaterialFunction /BRCosmetics/Accessories/FORT_Backpacks/Backpack_M_MED_PintPiano/Materials/MF_PintPiano_Backpack
graph {nodes [
  n1 = VertexInterpolator(VS=LocalPosition() {localOrigin=InstancePreSkinning}.XYZ)
  n2 = NamedRerouteDeclaration(None=SphereMask(
    A=NamedRerouteDeclaration(None=@n1.PS) {name=PreSkinPos, nodeColor=<rgba 0 0.7891 1 1>}
    B=VectorParameter() {parameterName=MaskPos, defaultValue=<rgba 0 0 0 1>}.RGB
    Radius=ScalarParameter() {defaultValue=100, parameterName=MaskRadius}
    Hardness=ScalarParameter() {parameterName=MaskHardness}
   )) {name=SphereMask, nodeColor=<rgba 1 0.3516 0 1>}
  n3 = ScalarParameter() {defaultValue=1, parameterName=MaskPower}
  n4 = ScalarParameter() {defaultValue=5, parameterName=FlickerSpeed}
  n5 = Multiply(A=Time(), B=@n4)
  n6 = Multiply(A=Time(), B=Multiply(A=@n4) {constB=0.4338})
  n7 = Multiply(A=ConstantBiasScale(None=Sine(None=@n5)), B=ConstantBiasScale(None=Sine(None=@n6)))
  n8 = Multiply(
   A=Multiply(A=Power(Base=@n2, Exp=@n3), B=LinearInterpolate(Alpha=@n7) {constA=0.05})
   B=VectorParameter() {parameterName=MaskColor, defaultValue=<rgba 0 1 0.5833 1>}.RGB
  )
  n9 = Multiply(
   A=Power(Base=NamedRerouteUsage(), Exp=ScalarParameter() {parameterName=ReactivityGlowPower})
   B=VectorParameter() {parameterName=ReactivityGlowColor, defaultValue=<rgba 0 1 0.5833 1>}.RGB
  )
  n10 = Divide(A=Subtract(A=Time(), B=ScalarParameter() {parameterName=TimeStart}), B=ScalarParameter() {defaultValue=5, parameterName=ReactivityDuration})
  n11 = Floor(None=@n10)
  n12 = Constant() {r=2}
  n13 = Frac(None=@n10)
  n14 = Power(Base=Abs(None=ConstantBiasScale(None=@n13) {bias=-0.5, scale=2}))
  n15 = Abs(None=ConstantBiasScale(None=Saturate(None=@n10)) {bias=-0.5, scale=2})
  n16 = StaticSwitchParameter(True=Multiply(A=Modulo!(A=@n11, B=@n12), B=OneMinus(None=@n14)), False=OneMinus(None=Power(Base=@n15))) {parameterName=DebugReactivity}
  n17 = ScalarParameter() {defaultValue=1, parameterName=MaskExponent}
  n18 = Multiply(A=Multiply(A=LinearInterpolate(A=@n8, B=@n9, Alpha=@n16), B=NamedRerouteUsage()), B=OneMinus(None=Fresnel(ExponentIn=@n17)))
  n19 = NamedRerouteUsage()
  n20 = Saturate(None=Add(A=ComponentMask(None=@n19) {g=true}, B=Constant() {r=10})) {desc=Mask glass panel}
  FunctionOutput(None=SetMaterialAttributes!(MaterialAttributes=NamedRerouteUsage(), EmissiveColor=Multiply(A=@n18, B=@n20)) {attributeSetTypes [#1]})
  NamedRerouteDeclaration(None=FunctionInput() {inputName=InMat, inputType=FunctionInput_MaterialAttributes}) {name=InMat, nodeColor=<rgba 1 0 0.3516 1>}
  NamedRerouteDeclaration(None=ComponentMask(None=FunctionInput() {inputName=InMask, sortPriority=10}) {r=true}) {name=InMask, nodeColor=<rgba 0 1 0.3672 1>}
 ]}
