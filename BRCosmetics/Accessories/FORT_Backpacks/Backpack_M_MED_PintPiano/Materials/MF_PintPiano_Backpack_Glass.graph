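# Material function for the PintPiano backpack glass that overrides the emissive of incoming material attributes with an animated energy glow.
# Sphere masks and a panel mask on pre-skinned positions shape noise-distorted, scrolling sine rings and a panel glow.
# A timed reactivity pulse from TimeStart and duration blends toward a separate glow color, with a static switch for a repeating debug pulse.
# Inverted Fresnel fades the result at grazing angles.
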MaterialFunction /BRCosmetics/Accessories/FORT_Backpacks/Backpack_M_MED_PintPiano/Materials/MF_PintPiano_Backpack_Glass
graph {nodes [
  n1 = NamedRerouteUsage()
  n2 = VectorParameter() {group=LoopGlow, parameterName=MaskColor, defaultValue=<rgba 0 1 0.5833 1>}
  n3 = NamedRerouteUsage()
  n4 = NamedRerouteUsage()
  n5 = Power(Base=NamedRerouteUsage(), Exp=ScalarParameter() {group=Reactivity, parameterName=ReactivityGlowPower})
  n6 = Divide(
   A=Subtract(A=Time(), B=ScalarParameter() {group=Reactivity, parameterName=TimeStart})
   B=ScalarParameter() {defaultValue=5, group=Reactivity, parameterName=ReactivityDuration}
  )
  n7 = Modulo!(A=Floor(None=@n6), B=Constant() {r=2})
  n8 = ConstantBiasScale(None=Frac(None=@n6)) {bias=-0.5, scale=2}
  n9 = OneMinus(None=Power(Base=Abs(None=@n8)))
  n10 = Saturate(None=@n6)
  n11 = Power(Base=Abs(None=ConstantBiasScale(None=@n10) {bias=-0.5, scale=2}))
  n12 = LinearInterpolate(
   A=Multiply(A=Multiply(A=@n1, B=@n2.RGB), B=@n3)
   B=Multiply(
    A=LinearInterpolate(A=@n4, B=@n5, Alpha=@n3)
    B=VectorParameter() {group=Reactivity, parameterName=ReactivityGlowColor, defaultValue=<rgba 0 1 0.5833 1>}.RGB
   )
   Alpha=StaticSwitchParameter(True=Multiply(A=@n7, B=@n9), False=OneMinus(None=@n11)) {parameterName=DebugReactivity}
  )
  n13 = OneMinus(None=Fresnel(ExponentIn=ScalarParameter() {defaultValue=1, group=LoopGlow, parameterName=MaskExponent}))
  FunctionOutput(None=SetMaterialAttributes!(MaterialAttributes=NamedRerouteUsage(), EmissiveColor=Multiply(A=@n12, B=@n13)) {attributeSetTypes [#1]})
  NamedRerouteDeclaration(None=FunctionInput() {inputName=InMat, inputType=FunctionInput_MaterialAttributes}) {name=InMat, nodeColor=<rgba 1 0 0.3516 1>}
  n14 = NamedRerouteUsage()
  n15 = Constant3Vector() {constant=<rgba 1.3 1 1 0>}
  n16 = OneMinus(None=SphereMask(A=Multiply(A=@n14, B=@n15), B=Constant3Vector() {constant=<rgba 0 -11 -0.5 0>}, Radius=Constant() {r=13}) {hardnessPercent=0})
  n17 = Constant() {r=5}
  NamedRerouteDeclaration(None=Multiply(A=Power(Base=@n16, Exp=@n17), B=Constant() {r=0.1})) {name=PanelGlow, nodeColor=<rgba 0 0.9531 1 1>}
  n18 = NamedRerouteDeclaration(None=VertexInterpolator(VS=LocalPosition() {localOrigin=InstancePreSkinning}.XYZ).PS) {name=PreSkinPos, nodeColor=<rgba 0 1 0.6016 1>}
  n19 = NamedRerouteDeclaration(None=VectorParameter() {group=LoopGlow, parameterName=MaskPos, defaultValue=<rgba 0 0 0 1>}.RGB) {name=MaskPos, nodeColor=<rgba 0.9219 0 1 1>}
  n20 = NamedRerouteDeclaration(None=ScalarParameter() {defaultValue=100, group=LoopGlow, parameterName=MaskRadius}) {name=Radius, nodeColor=<rgba 1 0.2812 0 1>}
  n21 = ScalarParameter() {group=LoopGlow, parameterName=MaskHardness}
  NamedRerouteDeclaration(None=Power(Base=SphereMask(A=@n18, B=@n19, Radius=@n20, Hardness=@n21), Exp=ScalarParameter() {defaultValue=1, group=LoopGlow, parameterName=MaskPower})) {name=SphereMask, nodeColor=<rgba 0 0.0391 1 1>}
  n22 = NamedRerouteUsage()
  n23 = Multiply(A=@n22) {constB=1.25}
  n24 = Add(A=NamedRerouteUsage(), B=Time())
  n25 = ScalarParameter() {defaultValue=0.5, group=LoopGlow, parameterName=Distortion Size}
  n26 = Multiply(
   A=Noise!(World Position=Multiply(A=@n24, B=@n25) {constB=0.5}, FilterWidth)
   B=ScalarParameter() {defaultValue=0.1, group=LoopGlow, parameterName=Distortion Intensity}
  ) {constB=0.1}
  n27 = Time()
  n28 = ScalarParameter() {defaultValue=-1, group=LoopGlow, parameterName=EnergySpeed}
  n29 = Sine(None=Add(A=Add(A=@n23, B=@n26), B=Multiply(A=@n27, B=@n28)))
  n30 = LinearInterpolate(Alpha=Power(Base=ConstantBiasScale(None=@n29), Exp=ScalarParameter() {defaultValue=1, group=LoopGlow, parameterName=Rings Power}) {constExponent=5}) {constA=0.3}
  n31 = SmoothStep(Value=NamedRerouteUsage()) {constMax=0.3}
  NamedRerouteDeclaration(None=Multiply(A=Multiply(A=@n30, B=@n31), B=ScalarParameter() {defaultValue=3, group=LoopGlow, parameterName=EnergyIntensity})) {name=EnergyMask, nodeColor=<rgba 0 1 0.7656 1>}
  n32 = ComponentMask(None=NamedRerouteUsage()) {g=true}
  n33 = Constant() {r=10}
  NamedRerouteDeclaration(None=Saturate(None=Add(A=@n32, B=@n33)) {desc=Mask glass panel}) {name=PanelMask, nodeColor=<rgba 1 0.7266 0 1>}
  Power(Base=@n22) {constExponent=1.5}
 ]}
